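; BP_MeshNetworkStatusFlare is a prototype actor that shows or hides itself to indicate whether a mesh network client is connected to root.
; On BeginPlay it waits asynchronously for the mesh network to be ready and continues only for the Client node type.
; It then binds to the subsystem's root connection change event and immediately applies the current state.
; Each update sets the actor's hidden state and prints a debug message.
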(event Custom|ConnectionToRootChanged
  (params (bConnected :type Boolean))
  (if bConnected
    (Rendering|SetActorHiddenInGame false)
    (Development|PrintString "BP_MeshNetworkStatusFlare - Mesh Network - CONNECTED to Root")
    (else
      (Rendering|SetActorHiddenInGame true)
      (Development|PrintString "BP_MeshNetworkStatusFlare - Mesh Network - NOT connected to Root"))))

(event EventBeginPlay
  (bind _returnvalue (GameInstanceSubsystems|GetMeshNetworkSubsystem))
  (bind _nodetype (Class|FortAsyncActionMeshNetworkReady|MeshNetworkReadyAsync)
    (:OnReady
      (switch Utilities|FlowControl|Switch|SwitchonEMeshNetworkNodeType _nodetype
        (:Client
          (MeshNetworkSubsystem|BindEventtoOnConnectedtoRootChanged _returnvalue (EventDispatchers|CreateEvent :!event Custom|ConnectionToRootChanged))
          (CallFunction|ConnectiontoRootChanged (MeshNetworkSubsystem|GetConnectedtoRoot _returnvalue)))))))

(fn ConstructionScript)
